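# Sound cue that plays looping, randomized ghost foley for the LycheeNickel emote.
# A mixer boosts and blends two independently looping branches: three main ghost waves and eight "Lyr" ghost variants.
# Each branch waits a random delay, wider for the variants, then plays an equally weighted random wave through a modulator, giving non-repeating texture.
# Emote foley attenuation, sound class and concurrency settings govern playback.
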SoundCue /BRCosmetics/Sounds/Emotes/LycheeNickel/SC_Emote_Foley_LycheeNickel_Ghost
root=Mixer_0
nodes [
 {id=Delay_1, children [Modulator_0], properties {delayMax=1.5, delayMin=1}}
 {id=Delay_2, children [Modulator_1], properties {delayMax=3, delayMin=1}}
 {id=Looping_0, children [Delay_1]}
 {id=Looping_1, children [Delay_2]}
 {id=Mixer_0, children [Looping_0,Looping_1], properties {inputVolume [2,1.75]}}
 {id=Modulator_0, children [Random_0]}
 {id=Modulator_1, children [Random_1]}
 {id=Random_0, children [WavePlayer_0,WavePlayer_1,WavePlayer_2], properties {weights [1,1,1]}}
 {
  id=Random_1
  children [WavePlayer_3, WavePlayer_4, WavePlayer_5, WavePlayer_6, WavePlayer_7, WavePlayer_8, WavePlayer_9, WavePlayer_10]
  properties {weights [1,1,1,1,1,1,1,1]}
 }
 {id=WavePlayer_0, properties {soundWaveAssetPtr=&/IP_Emote_LycheeNickel_Ghost_01}}
 {id=WavePlayer_1, properties {soundWaveAssetPtr=&/IP_Emote_LycheeNickel_Ghost_02}}
 {id=WavePlayer_10, properties {soundWaveAssetPtr=&/IP_Emote_LycheeNickel_Ghost_Lyr_08}}
 {id=WavePlayer_2, properties {soundWaveAssetPtr=&/IP_Emote_LycheeNickel_Ghost_03}}
 {id=WavePlayer_3, properties {soundWaveAssetPtr=&/IP_Emote_LycheeNickel_Ghost_Lyr_01}}
 {id=WavePlayer_4, properties {soundWaveAssetPtr=&/IP_Emote_LycheeNickel_Ghost_Lyr_02}}
 {id=WavePlayer_5, properties {soundWaveAssetPtr=&/IP_Emote_LycheeNickel_Ghost_Lyr_03}}
 {id=WavePlayer_6, properties {soundWaveAssetPtr=&/IP_Emote_LycheeNickel_Ghost_Lyr_04}}
 {id=WavePlayer_7, properties {soundWaveAssetPtr=&/IP_Emote_LycheeNickel_Ghost_Lyr_05}}
 {id=WavePlayer_8, properties {soundWaveAssetPtr=&/IP_Emote_LycheeNickel_Ghost_Lyr_06}}
 {id=WavePlayer_9, properties {soundWaveAssetPtr=&/IP_Emote_LycheeNickel_Ghost_Lyr_07}}
]
properties {
 attenuationSettings=&Fort_Emotes_Athena_Att
 audioPropertiesSheet=/BRCosmetics/Sounds/Settings/PropertySheets/Emotes/APS_Cosmetic_Emote_Foley
 maxDistance=1600
 soundClassObject=/Game/Sounds/FortSoundClassesAndModes/Fort_Foley_Emote
 volumeMultiplier=1
 concurrencySet [/Game/Sounds/Concurrency_Rules/Athena_EmoteFoley_Conc]
}
